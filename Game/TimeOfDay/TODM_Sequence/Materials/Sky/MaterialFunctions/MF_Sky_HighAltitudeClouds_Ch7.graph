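# Sky material function that composites lit, animated high-altitude clouds over an input color, limited by a horizon mask.
# Two counter-panning texture samples on a spherically distorted view vector build cloud opacity and normals, with zenith fading scaled by VolumetricsEnabled custom primitive data.
# A planar-projected overcast layer, driven by the OvercastSky collection parameter, supplies additional alpha and normals.
# Atmosphere light direction shades the clouds, and a halo mask boosts color and opacity.
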MaterialFunction /Game/TimeOfDay/TODM_Sequence/Materials/Sky/MaterialFunctions/MF_Sky_HighAltitudeClouds_Ch7
graph {
 properties {bExposeToLibrary=true, libraryCategoriesText [Misc,Coordinates]}
 nodes [
  n1 = SkyAtmosphereLightDirection()
  n2 = OneMinus(None=Abs(None=ComponentMask(None=@n1) {b=true}))
  n3 = DotProduct(
   A=LinearInterpolate(
    A=MaterialFunctionCall!(Multiplier, Offset) {materialFunction=/Game/TimeOfDay/TODM_Sequence/Materials/Sky/MF_MirroredCameraVector}.Result
    B=@n1
    Alpha=Multiply(A=@n2, B=@n2)
   )
   B=NamedRerouteUsage()
  )
  n4 = Saturate(None=Add(A=Multiply(A=@n3) {constB=0.5}) {constB=0.5})
  n5 = NamedRerouteUsage()
  n6 = ScalarParameter() {defaultValue=1.2, parameterName=Clouds Halo Light Boost}
  n7 = Multiply(A=Multiply(A=@n4, B=@n4), B=Add(A=Multiply(A=@n5, B=@n6) {constB=2}))
  n8 = LinearInterpolate(
   A=ComponentMask(None=VectorParameter() {parameterName=Clouds Shadow Color, defaultValue=<rgba 0 0 0 1>}.RGB) {b=true, g=true, r=true}
   B=ComponentMask(None=VectorParameter() {parameterName=Clouds Light Color, defaultValue=<rgba 0 0 0 1>}.RGB) {b=true, g=true, r=true}
   Alpha=ReflectionCapturePassSwitch!(Default=@n7, Reflection=@n7)
  )
  n9 = Multiply(A=@n8, B=ScalarParameter() {defaultValue=3, parameterName=Clouds Halo Color Boost}) {constB=12}
  n10 = LinearInterpolate(A=@n8, B=Multiply(A=@n9, B=@n9), Alpha=NamedRerouteUsage())
  n11 = NamedRerouteUsage()
  n12 = ReflectionCapturePassSwitch(Default=@n11, Reflection=@n11)
  n13 = FunctionInput() {bUsePreviewValueAsDefault=true, inputName=Horizon Mask, inputType=FunctionInput_Scalar, sortPriority=2}
  FunctionOutput(None=LinearInterpolate(A=FunctionInput() {bUsePreviewValueAsDefault=true, inputName=Color}, B=Multiply(A=@n10) {constB=4}, Alpha=Multiply(A=@n12, B=@n13))) {outputName=Color}
  n14 = TextureSample(UVs=MaterialFunctionCall(Multiplier=Constant() {r=-1}) {materialFunction=&MF_MirroredCameraVector}.Result) {samplerType=SAMPLERTYPE_LinearColor, texture=/Game/TimeOfDay/CloudsAndStorm/Textures/T_CloudsReflection}
  LinearInterpolate(A=@n7, B=@n14.R, Alpha=NamedRerouteDeclaration(None=@n14.B) {name=Clouds Mask, nodeColor=<rgba 0 0.4141 1 1>})
  Max(A=@n11, B=Multiply(A=NamedRerouteUsage()) {constB=0.5})
  n15 = MaterialFunctionCall() {materialFunction=&MF_MirroredCameraVector}
  n16 = ComponentMask(None=@n15.Result) {b=true}
  n17 = ScalarParameter() {defaultValue=0.2, parameterName=Clouds Spherical Distribution}
  n18 = Multiply(A=@n17) {constB=-1}
  n19 = Divide(B=Multiply(A=Add(A=@n16, B=@n18) {constB=-0.2}, B=OneMinus(None=@n17)) {constB=0.8})
  n20 = Multiply(A=ComponentMask(None=Multiply(A=@n19, B=@n15.Result)) {g=true, r=true}, B=ScalarParameter() {defaultValue=0.4, parameterName=Clouds Tiling})
  n21 = Time()
  n22 = Constant2Vector() {g=0.0064, r=0.0048}
  n23 = TextureSample(UVs=Add(A=Multiply(A=@n20) {constB=0.33}, B=Multiply(A=@n21, B=@n22))) {samplerType=SAMPLERTYPE_LinearColor, texture=/Game/TimeOfDay/TODM_Sequence/Materials/Sky/Textures/T_Ch6_HighAltitudeClouds}
  n24 = Time()
  n25 = Constant2Vector() {g=-0.0064, r=0.0064}
  n26 = TextureSample(UVs=Add(A=Multiply(A=@n20) {constB=-0.255}, B=Multiply(A=@n24, B=@n25))) {samplerType=SAMPLERTYPE_LinearColor, texture=&T_Ch6_HighAltitudeClouds}
  n27 = Saturate(None=Multiply(A=@n23.A, B=@n26.A))
  n28 = Multiply(A=@n27, B=@n27)
  n29 = Multiply(A=@n28, B=@n28)
  n30 = ComponentMask(None=CameraVectorWS()) {b=true}
  n31 = Multiply(
   A=Multiply(A=Multiply(A=@n30) {constB=-1}, B=ScalarParameter() {bUseCustomPrimitiveData=true, parameterName=VolumetricsEnabled})
   B=ScalarParameter() {defaultValue=6, parameterName=HideZenithClouds}
  )
  n32 = ScalarParameter() {defaultValue=0.9704, parameterName=HideZenithCloudsOffset}
  n33 = OneMinus(None=Saturate(None=Subtract(A=@n31, B=@n32)))
  n34 = Multiply(A=@n33, B=@n33)
  n35 = ScalarParameter() {defaultValue=2.5, parameterName=Clouds Opacity}
  n36 = ScalarParameter() {defaultValue=22, parameterName=Clouds Halo Opacity}
  n37 = NamedRerouteDeclaration(None=Multiply(
    A=FunctionInput() {bUsePreviewValueAsDefault=true, inputName=Halo Mask, inputType=FunctionInput_Scalar, sortPriority=1}
    B=ScalarParameter() {parameterName=Clouds Halo}
   )) {name=Halo, nodeColor=<rgba 0.1484 1 0 1>}
  n38 = Saturate(None=Multiply(A=Multiply(A=@n29, B=@n34) {constB=0}, B=LinearInterpolate(A=@n35, B=@n36, Alpha=@n37) {constA=2, constB=20}) {constB=10})
  n39 = Add(A=Min(A=@n38, B=ScalarParameter() {defaultValue=0.8, parameterName=Clouds Max Opacity}), B=NamedRerouteUsage())
  n40 = Add(A=Multiply(A=@n38) {constB=2}) {constB=-0.1}
  NamedRerouteDeclaration(None=Saturate(None=ReflectionCapturePassSwitch(Default=@n39, Reflection=@n40))) {name=Mask, nodeColor=<rgba 0 1 0.6719 1>}
  n41 = Add(A=ComponentMask(None=@n23.RGB) {g=true, r=true}) {constB=-0.5}
  n42 = Add(A=ComponentMask(None=@n26.RGB) {g=true, r=true}) {constB=-0.5}
  n43 = Add(A=AppendVector(A=Multiply(A=@n41) {constB=2}, B=@n23.A), B=AppendVector(A=Multiply(A=@n42) {constB=-2}, B=@n26.A))
  n44 = NamedRerouteUsage()
  n45 = NamedRerouteUsage()
  NamedRerouteDeclaration(None=ReflectionCapturePassSwitch(Default=LinearInterpolate(A=@n43, B=@n44, Alpha=@n45), Reflection=@n43)) {name=Normals, nodeColor=<rgba 0.5703 1 0 1>}
  n46 = Add(A=Constant3Vector() {constant=<rgba 0 0 200000 0>}, B=ActorPositionWS() {originType=CameraRelative})
  n47 = WorldPosition() {worldPositionShaderOffset=WPT_CameraRelative}
  n48 = Divide(A=ComponentMask(None=@n46) {b=true}, B=Add(A=@n47.Z) {constB=500000})
  n49 = Multiply(A=Subtract(A=Multiply(A=@n48, B=@n47.XY), B=ComponentMask(None=@n46) {g=true, r=true})) {constB=0}
  n50 = Time()
  n51 = Constant2Vector() {g=0.02, r=0.01}
  n52 = TextureSample(UVs=Add(A=Multiply(A=@n49) {constB=0.5}, B=Multiply(A=@n50, B=@n51))) {samplerType=SAMPLERTYPE_LinearColor, texture=/Game/TimeOfDay/TODM_Sequence/Materials/Sky/Textures/T_Ch6_OvercastClouds}
  n53 = Time()
  n54 = Constant2Vector() {g=-0.01, r=-0.02}
  n55 = TextureSample(UVs=Add(A=Multiply(A=@n49) {constB=-0.8}, B=Multiply(A=@n53, B=@n54))) {samplerType=SAMPLERTYPE_LinearColor, texture=&T_Ch6_OvercastClouds}
  n56 = Saturate(None=CollectionParameter() {collection=&FortniteMaterialParameters, parameterName=OvercastSky})
  n57 = Saturate(None=Add(A=Multiply(A=@n52.B, B=@n55.B), B=LinearInterpolate(Alpha=@n56) {constA=-0.6}))
  NamedRerouteDeclaration(None=Multiply(A=@n57, B=@n57)) {name=Overcast Alpha, nodeColor=<rgba 1 0.1641 0 1>}
  n58 = Add(A=@n55.RGB) {constB=-0.5}
  n59 = Constant3Vector() {constant=<rgba -1 -1 1 0>}
  n60 = Multiply(A=Subtract(A=@n55.B, B=Multiply(A=@n52.B) {constB=0.6})) {constB=2.4}
  NamedRerouteDeclaration(None=LinearInterpolate(A=Add(A=@n52.RGB) {constB=-0.5}, B=Multiply(A=@n58, B=@n59), Alpha=Saturate(None=@n60))) {name=Overcast Normals, nodeColor=<rgba 0 0.5781 1 1>}
 ]
}
