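# Sky material function that builds the reflection-pass color from an input color and a vertex offset for the sky mesh.
# The input is desaturated, boosted by vertical view angle, then blended by that angle into a ground texture projected along the camera vector and tinted by time of day.
# A sun-relative vertex-normal term modulates the result.
# The WPO output squashes vertices below world Z zero almost flat onto that plane.
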MaterialFunction /Game/TimeOfDay/TODM_Sequence/Materials/Sky/MaterialFunctions/MF_Sky_Reflection
graph {nodes [
  n1 = Desaturation(
   None=FunctionInput() {bUsePreviewValueAsDefault=true, inputName=Color}
   Fraction=ScalarParameter() {defaultValue=0.25, group=Reflection Pass, parameterName=Reflection Desaturation}
  )
  n2 = ScalarParameter() {defaultValue=0.25, group=Reflection Pass, parameterName=Reflection Secondary Desaturation}
  n3 = CameraVectorWS()
  n4 = Subtract(A=Abs(None=ComponentMask(None=@n3) {b=true})) {constB=0.1}
  n5 = Saturate(None=Multiply(A=Abs(None=@n4)) {constB=2})
  n6 = OneMinus(None=Multiply(A=@n5, B=@n5))
  n7 = OneMinus(None=Multiply(A=@n6, B=@n6))
  n8 = Multiply(
   A=Add(A=Desaturation(None=@n1, Fraction=@n2), B=VectorParameter() {parameterName=Reflection Additive, defaultValue=<rgba 0 0 0 1>}.RGB)
   B=Multiply(A=LinearInterpolate(Alpha=@n7) {constA=0.3}, B=ScalarParameter() {defaultValue=5, parameterName=Reflection Boost}) {constA=0.5}
  ) {constB=4}
  n9 = CameraVectorWS()
  n10 = ComponentMask(None=@n9) {b=true}
  n11 = ComponentMask(None=Multiply(A=@n9, B=Divide(B=@n10))) {g=true, r=true}
  n12 = Desaturation(
   None=TextureSample(UVs=Multiply(A=@n11) {constB=-0.1}) {texture=/Game/TimeOfDay/TODM/BR/s11/GroundMesh/T_ReflectionGround}.RGB
   Fraction=Constant() {r=0.7}
  )
  n13 = VectorParameter() {parameterName=ColorMorning, defaultValue=<rgba 0.4856 0.3925 0.7 1>}
  n14 = VectorParameter() {parameterName=ColorDay, defaultValue=<rgba 0.9 0.5625 0.7242 1>}
  n15 = VectorParameter() {parameterName=ColorEvening, defaultValue=<rgba 0.5078 0.4803 0.7 1>}
  n16 = VectorParameter() {parameterName=ColorNight, defaultValue=<rgba 0.5224 0.5681 0.85 1>}
  n17 = Multiply(A=MaterialFunctionCall(
    Color_Morning=Multiply(A=@n12, B=@n13.RGB)
    Color_Day=Multiply(A=@n12, B=@n14.RGB)
    Color_Evening=Multiply(A=@n12, B=@n15.RGB)
    Color_Night=Multiply(A=@n12, B=@n16.RGB)
   ) {materialFunction=&Color_Time_of_Day}.Result) {constB=2}
  n18 = Saturate(None=ComponentMask(None=CameraVectorWS()) {b=true})
  n19 = OneMinus(None=Power(Base=OneMinus(None=@n18)) {constExponent=8})
  n20 = SkyAtmosphereLightDirection()
  n21 = Constant3Vector() {constant=<rgba 0 0 1 0>}
  n22 = Normalize(VectorInput=ComponentMask(None=CrossProduct(A=@n20, B=@n21)) {g=true, r=true})
  n23 = Constant()
  n24 = Multiply(A=DotProduct(A=AppendVector(A=@n22, B=@n23), B=VertexNormalWS()), B=ScalarParameter() {parameterName=Reflection Corner Break}) {constB=0.4}
  FunctionOutput(None=Multiply(A=LinearInterpolate(A=@n8, B=@n17, Alpha=@n19), B=Add(A=@n24))) {outputName=Reflection Color}
  n25 = WorldPosition()
  n26 = Constant3Vector() {constant=<rgba 0 0 -0.999 0>}
  n27 = Multiply(A=@n25.Z) {constB=-1}
  FunctionOutput(None=Multiply(A=Multiply(A=@n25.XYZ, B=@n26), B=Saturate(None=@n27))) {outputName=Reflection WPO, sortPriority=1}
  n28 = Saturate(None=ComponentMask(None=CameraVectorWS()) {b=true})
  n29 = OneMinus(None=@n28)
  n30 = Multiply(A=@n29, B=@n29)
  n31 = Multiply(A=@n28, B=Multiply(A=@n30, B=@n30))
  Multiply(A=Add(A=Multiply(A=@n31) {constB=20}))
 ]}
