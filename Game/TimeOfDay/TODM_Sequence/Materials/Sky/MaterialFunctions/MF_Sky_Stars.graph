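# Material function for the time-of-day sky that outputs twinkling stars from a stars texture.
# The texture's red channel is sampled static and time-panned, summed, then masked by the static green channel to animate brightness.
# Two squarings raise the result to the fourth power for contrast.
# It is then scaled by the boosted Stars color parameter, black by default, and the HorizonMask input.
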MaterialFunction /Game/TimeOfDay/TODM_Sequence/Materials/Sky/MaterialFunctions/MF_Sky_Stars
graph {nodes [
  n1 = FunctionInput(Preview=TextureCoordinate() {uTiling=4.25, vTiling=4.25}) {bUsePreviewValueAsDefault=true, inputName=UV, inputType=FunctionInput_Vector2}
  n2 = Add(A=@n1, B=Multiply(A=Time(), B=Constant2Vector() {g=0.27, r=0.25}) {constB=0.2})
  n3 = TextureSample(UVs=@n1) {samplerType=SAMPLERTYPE_LinearColor, texture=/Game/TimeOfDay/TODM_Sequence/Materials/Sky/Textures/T_Ch6_Stars}
  n4 = Multiply(
   A=Add(
    A=TextureSample(UVs=@n2) {samplerType=SAMPLERTYPE_LinearColor, texture=&T_Ch6_Stars}.R
    B=@n3.R
   )
   B=@n3.G
  )
  n5 = Multiply(A=@n4, B=@n4)
  n6 = Multiply(A=@n5, B=@n5)
  n7 = Multiply(A=ComponentMask(None=VectorParameter() {group=Stars, parameterName=Stars, defaultValue=<rgba 0 0 0 1>}.RGB) {b=true, g=true, r=true}) {constB=15}
  n8 = FunctionInput() {bUsePreviewValueAsDefault=true, inputName=HorizonMask, inputType=FunctionInput_Scalar, previewValue=<1 1 1 1>}
  FunctionOutput(None=Multiply(A=Multiply(A=@n6, B=@n7), B=@n8))
 ]}
